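; BeastGirl pawn Blueprint that exposes transformation and dive state for animation.
; GetAnimData returns the transform-animation flag, replicated transformation data from the stand-in pawn and the current dive state.
; Total transformation duration is read from a scalable float.
; Dive state is set directly or reset to Inactive by a cancellable half-second timer.
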(fn GetAnimData
  (return (Variables|Default|GetbIsInTransformAnimation) (Variables|BeastGirlStandInPawn|GetReplicatedTransformationData) (Variables|Default|GetCurrentDiveState)))

(fn GetTotalTransformationDuration
  (return (ScalableFloat|GetValueatLevel (Variables|Default|GetHF_TransformationDuration))))

(fn ResetDiveState
  (Variables|Default|SetCurrentDiveState "Inactive"))

(fn SetDiveState
  (params (DiveState :type EBeastGirlDiveState%20Enum :default "Inactive"))
  (Variables|Default|SetCurrentDiveState DiveState))

(fn StartInactiveDiveTimer
  (params (Active :type Boolean))
  (if Active
    (bind _returnvalue (Utilities|Time|SetTimerbyEvent (EventDispatchers|CreateEvent :!event ResetDiveState) 0.5))
    (Variables|Default|SetResetDiveStateTimerHandle _returnvalue)
    (else
      (Utilities|Time|ClearandInvalidateTimerbyHandle (Variables|Default|GetResetDiveStateTimerHandle)))))

(fn ConstructionScript)
