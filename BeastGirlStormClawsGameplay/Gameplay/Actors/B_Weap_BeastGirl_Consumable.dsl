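; Weapon actor for the BeastGirl consumable that adds a left-hand mesh and pom Niagara effects to the wielder.
; On equip, it snaps the left-hand mesh to the pawn's socket, activates idle FX and starts three async gameplay-tag count listeners.
; Parachuting hides the weapon mesh, while jamming or emoting hides the NS_BeastGirl_Idle_Poms1 effect.
; Unequip ends only the parachute listener, not the jamming or emote ones.
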(event EventBeginPlay)

(event EventTick
  (params (DeltaSeconds :type Float%20%28single-precision%29)))

(event FortWeapon|EventOnEquippedAndAttached
  (bind _as_fort_pawn (Variables|Default|GetAs%20Fort%20Pawn))
  (|CallParentFunction)
  (bind _asfort_pawn (Utilities|Casting|CastToFortPawn (Game|GetInstigator)))
  (Variables|Default|SetAs%20Fort%20Pawn _asfort_pawn)
  (Utilities|IsValid (Variables|/Script/Engine.Character|GetMesh _asfort_pawn)
    (:Is%20Valid
      (Transformation|AttachComponentToComponent (Variables|Default|GetLeftHandWeaponMesh) (Variables|/Script/Engine.Character|GetMesh _asfort_pawn) (Variables|Default|GetLeftHandAttachSocket) "SnapToTarget" "SnapToTarget" "KeepRelative" true)
      (Components|Activation|Activate (Variables|Default|GetNS_BeastGirl_Idle_Poms) true)
      (bind (_asynctaskproxy _tagcount) (Ability|Tasks|WaitGameplayTagCountChangedOnActor _as_fort_pawn "(TagName=\"MovementMode.Parachuting\")")
        (:then
          (Variables|Default|SetAsync%20Action%20Parachute _asynctaskproxy)
          (bind (_asynctaskproxy_2 _tagcount_2) (Ability|Tasks|WaitGameplayTagCountChangedOnActor _as_fort_pawn "(TagName=\"Gameplay.Action.Player.Jam.Jamming\")")
            (:then
              (Variables|Default|SetAsync%20Action%20Jamming _asynctaskproxy_2)
              (bind (_asynctaskproxy_3 _tagcount_3) (Ability|Tasks|WaitGameplayTagCountChangedOnActor _as_fort_pawn "(TagName=\"Gameplay.Action.Player.Emote\")")
                (:then
                  (Variables|Default|SetAsync%20Action%20Emote _asynctaskproxy_3))
                (:TagCountChanged
                  (Rendering|SetVisibility (Variables|Default|GetNS_BeastGirl_Idle_Poms1) (not (>= _tagcount_3 1)) false))))
            (:TagCountChanged
              (Rendering|SetVisibility (Variables|Default|GetNS_BeastGirl_Idle_Poms1) (not (>= _tagcount_2 1)) false))))
        (:TagCountChanged
          (Rendering|SetVisibility (Variables|Visual|GetWeaponMesh) (not (>= _tagcount 1)) false))))))

(event FortWeapon|EventOnUnEquippedAndDetached
  (|CallParentFunction)
  (Ability|Async|EndAction (Variables|Default|GetAsync%20Action%20Parachute)))

(fn ConstructionScript)
