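; Storm Claws weapon Blueprint that adds custom sprint behavior and drives sprint-related HUD visibility.
; BeginPlay casts the instigator to a pawn, retrying on a timer, then watches the BeastGirl sprinting tag count to set HUD element visibility.
; Equip finds the controller's TacticalSprint component and adds a custom sprint MME class.
; Unequip removes that class, clears the component reference and ends the tag watcher, which equip does not restart.
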(event Custom|ReCheckForPawn
  (bind _asfort_pawn (Utilities|Casting|CastToFortPawn (Game|GetInstigator))
    (:then
      (bind _output_get (Variables|Default|SetPlayerPawn _asfort_pawn))
      (bind (_asynctaskproxy _tagcount) (Ability|Tasks|WaitGameplayTagCountChangedOnActor _output_get (Variables|Default|GetBeastGirl_SprintingTag))
        (:then
          (Variables|Default|SetWaitTagCountAction _asynctaskproxy))
        (:TagCountChanged
          (HUD|SetHUDElementVisibilitywithContext (Fortnite|GetFortPlayerControllerfromActor (Variables|Default|GetPlayerPawn)) (Variables|Default|GetSprintTag) (>= _tagcount 1) "SprintActive" ""))))
    (:CastFailed
      (Utilities|Time|SetTimerbyEvent (EventDispatchers|CreateEvent :!event Custom|ReCheckForPawn) 0.25))))

(event EventBeginPlay
  (|CallParentFunction)
  (bind _asfort_pawn (Utilities|Casting|CastToFortPawn (Game|GetInstigator))
    (:then
      (bind _output_get (Variables|Default|SetPlayerPawn _asfort_pawn))
      (bind (_asynctaskproxy _tagcount) (Ability|Tasks|WaitGameplayTagCountChangedOnActor _output_get (Variables|Default|GetBeastGirl_SprintingTag))
        (:then
          (Variables|Default|SetWaitTagCountAction _asynctaskproxy))
        (:TagCountChanged
          (HUD|SetHUDElementVisibilitywithContext (Fortnite|GetFortPlayerControllerfromActor (Variables|Default|GetPlayerPawn)) (Variables|Default|GetSprintTag) (>= _tagcount 1) "SprintActive" ""))))
    (:CastFailed
      (Utilities|Time|SetTimerbyEvent (EventDispatchers|CreateEvent :!event Custom|ReCheckForPawn) 0.25))))

(event EventTick
  (params (DeltaSeconds :type Float%20%28single-precision%29)))

(event FortWeapon|EventOnEquippedAndAttached
  (bind _playerpawn (Variables|Default|GetPlayerPawn))
  (bind _returnvalue (Pawn|GetController _playerpawn))
  (Utilities|IsValid _returnvalue
    (:Is%20Valid
      (Variables|Default|SetLocal%20Sprint%20Component (Actor|GetComponentbyClass _returnvalue "/TacticalSprintGame/Gameplay/TacticalSprintControllerComponent.TacticalSprintControllerComponent_C"))
      (CallFunction|ToggleSprintMME true))))

(event FortWeapon|EventOnUnEquippedAndDetached
  (CallFunction|ToggleSprintMME false)
  (Ability|Async|EndAction (Variables|Default|GetWaitTagCountAction))
  (Variables|Default|SetLocal%20Sprint%20Component 0))

(fn ToggleSprintMME
  (params (Enabled :type Boolean))
  (Utilities|IsValid (Variables|Default|GetLocal%20Sprint%20Component)
    (:Is%20Valid
      (if Enabled
        (bind _customsprintmme (Variables|Default|GetCustomSprintMME)
          (:then
            (Sprint|AddCustomSprintMMEClass (Variables|Default|GetLocal%20Sprint%20Component) _customsprintmme))
          (:else))
        (else
          (bind _customsprintmme_2 (Variables|Default|GetCustomSprintMME)
            (:then
              (Sprint|RemoveCustomSprintMMEClass (Variables|Default|GetLocal%20Sprint%20Component) _customsprintmme_2))
            (:else)))))))

(fn ConstructionScript
  (|CallParentFunction))
